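# Sound cue that plays a looping mono water stream sound, filed under loot container foley and assigned to the Fort_Loot_Collect sound class.
# A looping wave player feeds a modulator that applies a fixed pitch raise, then a mixer that trims the level before the cue's volume boost.
# Attenuation overrides add distance falloff and listener focus, quieting it when out of focus.
# A concurrency override prevents new instances once the limit is reached.
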SoundCue /Game/Sounds/Foley_Loot/Containers/Water/Loops/Stream1_mono_Cue
root=Mixer_0
nodes [
 {id=Mixer_0, children [Modulator_1], properties {inputVolume [0.7]}}
 {id=Modulator_1, children [WavePlayer_21], properties {pitchMax=1.2, pitchMin=1.2}}
 {id=WavePlayer_21, properties {bLooping=true, soundWaveAssetPtr=&/Stream1_mono}}
]
properties {
 bOverrideAttenuation=true
 bOverrideConcurrency=true
 maxDistance=1300
 soundClassObject=&Fort_Loot_Collect
 volumeMultiplier=1.5
 attenuationOverrides {bEnableListenerFocus=true, falloffDistance=1000, nonFocusVolumeAttenuation=0.3, attenuationShapeExtents {x=300}}
 concurrencyOverrides {resolutionRule=PreventNew}
}
